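# Physics asset for the M_MED_Possession male medium player body, set up for three dangling tasset armor pieces (back, left, right).
# Each tasset capsule is constrained with tight swing and twist limits to a tiny master-bone capsule, which links to spine_02 with free rotation.
# Pelvis, lower spine and thigh capsules supply body volumes around the tassets.
# Spine joints rotate freely, while thigh-to-pelvis joints carry wider limits than the tassets.
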PhysicsAsset /BRCosmetics/Characters/Player/Male/Medium/Bodies/M_MED_Possession/Meshes/M_MED_Possession_Physics
bodies {
 dyn_tasset_bk [{shapeType=Capsule, length=19.0131, radius=12.1196, center=<15 -10 0>, rotation=<rot 90 0 -3>}]
 dyn_tasset_bk_MSTR [{shapeType=Capsule, length=0.505, radius=0.505, shapeName=dyn_tasset_bk_MSTR_capsule, rotation=<rot 0 -90 0>}]
 dyn_tasset_l [{shapeType=Capsule, length=21.2624, radius=8.4224, center=<15 0 10>, rotation=<rot 0 90.1394 -91.3574>}]
 dyn_tasset_l_MSTR [{shapeType=Capsule, length=0.505, radius=0.505, shapeName=dyn_tasset_l_MSTR_capsule, rotation=<rot 0 -90 0>}]
 dyn_tasset_r [{shapeType=Capsule, length=21.2624, radius=8.4224, center=<15 0 10>, rotation=<rot 0 90.1394 -91.3574>}]
 dyn_tasset_r_MSTR [{shapeType=Capsule, length=0.505, radius=0.505, shapeName=dyn_tasset_r_MSTR_capsule, rotation=<rot 0 -90 0>}]
 pelvis [{shapeType=Capsule, length=7.4249, radius=8.6205, center=<-3.1833 -0.8408 0>}]
 spine_01 [{shapeType=Capsule, length=4.38, radius=7.005, shapeName=spine_01_capsule, rotation=<rot 0 -90 0>}]
 spine_02 [{
   shapeType=Capsule
   length=6.1241
   radius=5.8817
   shapeName=spine_02_capsule
   center=<5.2644 1.1225 0.1054>
   rotation=<rot -2.1844 -90 2.9371>
  }]
 thigh_l [{shapeType=Capsule, length=24.8476, radius=7.4827, shapeName=thigh_l_capsule, center=<-22 1 0>, rotation=<rot 90 0 -1.6898>}]
 thigh_r [{shapeType=Capsule, length=25.2226, radius=7.1077, shapeName=thigh_r_capsule, center=<22 -1 0>, rotation=<rot 90 0 -1.6898>}]
}
constraints [
 {bone1Name=dyn_tasset_bk, bone2Name=dyn_tasset_bk_MSTR, swing1LimitDegrees=7, swing2LimitDegrees=1, twistLimitDegrees=1}
 {bone1Name=dyn_tasset_l, bone2Name=dyn_tasset_l_MSTR, swing1LimitDegrees=5, swing2LimitDegrees=10, twistLimitDegrees=3}
 {bone1Name=dyn_tasset_r, bone2Name=dyn_tasset_r_MSTR, swing1LimitDegrees=5, swing2LimitDegrees=10, twistLimitDegrees=3}
 {bone1Name=thigh_l, bone2Name=pelvis, swing1LimitDegrees=45, swing2LimitDegrees=45, twistLimitDegrees=45}
 {bone1Name=thigh_r, bone2Name=pelvis, swing1LimitDegrees=45, swing2LimitDegrees=45, twistLimitDegrees=45}
 {
  bone1Name=spine_01
  bone2Name=pelvis
  swing1LimitDegrees=45
  swing1Motion=Free
  swing2LimitDegrees=45
  swing2Motion=Free
  twistLimitDegrees=45
  twistMotion=Free
 }
 {
  bone1Name=spine_02
  bone2Name=spine_01
  swing1LimitDegrees=45
  swing1Motion=Free
  swing2LimitDegrees=45
  swing2Motion=Free
  twistLimitDegrees=45
  twistMotion=Free
 }
 {
  bone1Name=dyn_tasset_r_MSTR
  bone2Name=spine_02
  swing1LimitDegrees=45
  swing1Motion=Free
  swing2LimitDegrees=45
  swing2Motion=Free
  twistLimitDegrees=45
  twistMotion=Free
 }
 {
  bone1Name=dyn_tasset_l_MSTR
  bone2Name=spine_02
  swing1LimitDegrees=45
  swing1Motion=Free
  swing2LimitDegrees=45
  swing2Motion=Free
  twistLimitDegrees=45
  twistMotion=Free
 }
 {
  bone1Name=dyn_tasset_bk_MSTR
  bone2Name=spine_02
  swing1LimitDegrees=45
  swing1Motion=Free
  swing2LimitDegrees=45
  swing2Motion=Free
  twistLimitDegrees=45
  twistMotion=Free
 }
]
