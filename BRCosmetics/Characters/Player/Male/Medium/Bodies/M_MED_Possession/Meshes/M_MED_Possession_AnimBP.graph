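# Post-style AnimBlueprint for the male medium Possession body that drives its tasset (hip armour) bones on top of the pose copied from the attached parent mesh.
# A deform Control Rig, weighted bone constraints and a second Control Rig feed jump offsets, a pelvis-space RigidBody simulation, then corrective bone offsets.
# The EventGraph sets those alphas from crouch, jump, fall, skydive and backward-diagonal movement.
# An AnimDynamics chain sits unconnected to the output.
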AnimBlueprint /BRCosmetics/Characters/Player/Male/Medium/Bodies/M_MED_Possession/Meshes/M_MED_Possession_AnimBP
target_skeleton=&/M_MED_Possession_Skeleton
graphs [
 AnimGraph {nodes [
   n1 = Anim Dynamics - Bone: dyn_tasset_bk(Alpha=1, ExternalForce="0,0,0") {
    angularDampingOverride=0.998
    angularSpringConstant=150
    bAngularSpring=true
    bGravityOverrideInSimSpace=true
    bOverrideAngularDamping=true
    bUseGravityOverride=true
    bUsePlanarLimit=false
    simulationSpace=RootRelative
    alphaBoolBlend {blendOption=Linear}
    boundBone {boneName=dyn_tasset_bk}
    componentAppliedLinearAccClamp=<3000 3000 3000>
    componentLinearAccScale=<0.5 0.5 0.5>
    componentLinearVelScale=<0.5 0.5 0.5>
    gravityOverride {z=-980}
    physicsBodyDefinitions [{
      collisionType=InnerSphere
      sphereCollisionRadius=1
      boundBone {boneName=dyn_tasset_bk}
      boxExtents=<12 15 15>
      constraintSetup {twistAxis=AxisZ, angularLimitsMax=<0 5 0>, angularLimitsMin=<0 -5 -10>, angularTarget=<0 0 0>}
      localJointOffset=<1 0 0>
     }]
    relativeSpaceBone {boneName=root}
   }
   n2 = Anim Dynamics - Bone: dyn_tasset_l(ComponentPose=@n1, Alpha=0.5, ExternalForce="0,0,0") {
    angularDampingOverride=0.998
    angularSpringConstant=150
    bAngularSpring=true
    bGravityOverrideInSimSpace=true
    bOverrideAngularDamping=true
    bUseGravityOverride=true
    bUsePlanarLimit=false
    simulationSpace=RootRelative
    alphaBoolBlend {blendOption=Linear}
    boundBone {boneName=dyn_tasset_l}
    componentAppliedLinearAccClamp=<3000 3000 3000>
    componentLinearAccScale=<0.5 0.5 0.5>
    componentLinearVelScale=<0.5 0.5 0.5>
    gravityOverride {z=-980}
    physicsBodyDefinitions [{
      collisionType=InnerSphere
      sphereCollisionRadius=1
      boundBone {boneName=dyn_tasset_l}
      boxExtents=<20 15 15>
      constraintSetup {twistAxis=AxisZ, angularLimitsMax=<0 0 3>, angularLimitsMin=<0 -5 -3>, angularTarget=<0 0 0>}
      localJointOffset=<1 0 0>
     }]
    relativeSpaceBone {boneName=root}
   }
   Anim Dynamics - Bone: dyn_tasset_r(ComponentPose=@n2, Alpha=0.5, ExternalForce="0,0,0") {
    angularDampingOverride=0.998
    angularSpringConstant=150
    bAngularSpring=true
    bGravityOverrideInSimSpace=true
    bOverrideAngularDamping=true
    bUseGravityOverride=true
    bUsePlanarLimit=false
    simulationSpace=RootRelative
    alphaBoolBlend {blendOption=Linear}
    boundBone {boneName=dyn_tasset_r}
    componentAppliedLinearAccClamp=<3000 3000 3000>
    componentLinearAccScale=<0.5 0.5 0.5>
    componentLinearVelScale=<0.5 0.5 0.5>
    gravityOverride {z=-980}
    physicsBodyDefinitions [{
      collisionType=InnerSphere
      sphereCollisionRadius=1
      boundBone {boneName=dyn_tasset_r}
      boxExtents=<20 15 15>
      constraintSetup {twistAxis=AxisZ, angularLimitsMax=<0 0 3>, angularLimitsMin=<0 -5 -3>, angularTarget=<0 0 0>}
      localJointOffset=<1 0 0>
     }]
    relativeSpaceBone {boneName=root}
   }
   n3 = Control Rig(Source=Copy Pose From Mesh() {bUseAttachedParent=true}) {
    alphaBoolBlend {blendOption=Linear}
    controlRigAssetReference {blueprintRigClass=&/M_MED_DeformRig_possession_ControlRig.M_MED_DeformRig_possession_ControlRig_C}
   }
   n4 = "Constraint - dyn_tasset_fr_MSTR (PPPPPPPP)"!(
    ComponentPose=Local To Component(LocalPose=@n3)
    Alpha=1
    ConstraintWeights_0=0.3769
    ConstraintWeights_1:float=0.1874
    ConstraintWeights_2:float=0.2492
    ConstraintWeights_3:float=0.1647
    ConstraintWeights_4:float=0.0218
    ConstraintWeights_5:float=0
    ConstraintWeights_6:float=0
    ConstraintWeights_7:float=0
   ) -> (Pose) {
    alphaBoolBlend {blendOption=Linear}
    boneToModify {boneName=dyn_tasset_fr_MSTR}
    constraintSetup [
     {targetBone {boneName=pelvis}}
     {targetBone {boneName=spine_01}}
     {targetBone {boneName=spine_02}}
     {targetBone {boneName=spine_03}}
     {targetBone {boneName=spine_04}}
     {targetBone {boneName=deform_groin_l}}
     {targetBone {boneName=thigh_twist_01_r}}
     {targetBone {boneName=deform_groin_r}}
    ]
    constraintWeights [1,1,1,1,1,1,1,1]
   }
   n5 = "Constraint - dyn_tasset_bk_MSTR (R)"!(ComponentPose=@n4, bAlphaBoolEnabled:bool=Get Tasset BK() -> (Tasset BK:bool), ConstraintWeights_0=1) -> (Pose) {
    alphaInputType=Bool
    alphaBoolBlend {blendInTime=0.1, blendOption=Linear, blendOutTime=0.1}
    boneToModify {boneName=dyn_tasset_bk_MSTR}
    constraintSetup [{transformType=Rotation, targetBone {boneName=pelvis}}]
    constraintWeights [1]
   }
   n6 = Control Rig!(Source=Component To Local(ComponentPose=@n5), Alpha, skydive:bool=Get bIsSkyDiving(), Crouching:double=Get crouching() -> (crouching:double)) -> (Pose) {
    alphaBoolBlend {blendOption=Linear}
    controlRigAssetReference {blueprintRigClass=&/M_MED_Possession_CtrlRig.M_MED_Possession_CtrlRig_C}
   }
   n7 = Get jumping() -> (jumping:double)
   n8 = "Transform (Modify) Bone - Bone: dyn_tasset_bk_MSTR"(ComponentPose=Local To Component(LocalPose=@n6), Alpha=@n7, Translation="0,-1,0", Rotation="-0.000,-10.000,0.000", Scale="1,1,1") {
    rotationMode=BMM_Additive
    rotationSpace=BCS_BoneSpace
    translationMode=BMM_Additive
    translationSpace=BCS_BoneSpace
    alphaBoolBlend {blendOption=Linear}
    alphaScaleBiasClamp {bInterpResult=true, interpSpeedIncreasing=5}
    boneToModify {boneName=dyn_tasset_bk_MSTR}
    rotation=<rot 0 -10 0>
    translation {y=-1}
   }
   n9 = "Transform (Modify) Bone - Bone: dyn_tasset_l_MSTR"(ComponentPose=@n8, Alpha=@n7, Translation="1,0,1", Rotation="10,0,0", Scale="1,1,1") {
    rotationMode=BMM_Additive
    rotationSpace=BCS_BoneSpace
    translationMode=BMM_Additive
    translationSpace=BCS_BoneSpace
    alphaBoolBlend {blendOption=Linear}
    alphaScaleBiasClamp {bInterpResult=true, interpSpeedIncreasing=5}
    boneToModify {boneName=dyn_tasset_l_MSTR}
    rotation {pitch=10}
    translation {x=1, z=1}
   }
   n10 = "Transform (Modify) Bone - Bone: dyn_tasset_r_MSTR"(ComponentPose=@n9, Alpha=@n7, Translation="1,0,1", Rotation="10,0,0", Scale="1,1,1") {
    rotationMode=BMM_Additive
    rotationSpace=BCS_BoneSpace
    translationMode=BMM_Additive
    translationSpace=BCS_BoneSpace
    alphaBoolBlend {blendOption=Linear}
    alphaScaleBiasClamp {bInterpResult=true, interpSpeedIncreasing=5}
    boneToModify {boneName=dyn_tasset_r_MSTR}
    rotation {pitch=10}
    translation {x=1, z=1}
   }
   n11 = RigidBody!(ComponentPose=@n10, Alpha, OverrideWorldGravity:Vector="0,0,-980", ExternalForce) -> (Pose) {
    bClampLinearTranslationLimitToRefPose=true
    bOverrideWorldGravity=true
    overridePhysicsAsset=&/M_MED_Possession_Physics
    simulationSpace=BaseBoneSpace
    alphaBoolBlend {blendInTime=0.1, blendOption=Linear, blendOutTime=0.1}
    baseBoneRef {boneName=pelvis}
    componentAppliedLinearAccClamp=<3000 3000 3000>
    componentLinearAccScale=<0.3 0.3 0.3>
    componentLinearVelScale=<0.3 0.3 0.3>
    simSpaceSettings {
     maxAngularAcceleration=1000
     maxAngularVelocity=3
     maxLinearAcceleration=700
     maxLinearVelocity=500
     velocityScaleZ=0
     externalLinearDragV=<0.2 0.2>
    }
   }
   n12 = "Transform (Modify) Bone - Bone: dyn_tasset_l"(ComponentPose=@n11, Alpha=Get BK_RightMovement() -> (BK_RightMovement:float), Translation="-0.5,0,0.5", Rotation="5,0,5", Scale="1,1,1") {
    rotationMode=BMM_Additive
    rotationSpace=BCS_BoneSpace
    translationMode=BMM_Additive
    translationSpace=BCS_BoneSpace
    alphaBoolBlend {blendOption=Linear}
    alphaScaleBiasClamp {bInterpResult=true, interpSpeedIncreasing=5}
    boneToModify {boneName=dyn_tasset_l}
    rotation {pitch=5, roll=5}
    translation {x=-0.5, z=0.5}
   }
   n13 = "Transform (Modify) Bone - Bone: dyn_tasset_r"(ComponentPose=@n12, Alpha=Get BK_LeftMovement() -> (BK_LeftMovement:float), Translation="0,0,0", Rotation="5,0,-5", Scale="1,1,1") {
    rotationMode=BMM_Additive
    rotationSpace=BCS_BoneSpace
    translationMode=BMM_Additive
    translationSpace=BCS_BoneSpace
    alphaBoolBlend {blendOption=Linear}
    alphaScaleBiasClamp {bInterpResult=true, interpSpeedIncreasing=5}
    boneToModify {boneName=dyn_tasset_r}
    rotation {pitch=5, roll=-5}
   }
   n14 = "Transform (Modify) Bone - Bone: dyn_tasset_l"(ComponentPose=@n13, Alpha=Get Tasset fix() -> (Tasset fix:double), Translation="0,0,1", Rotation="0,0,0", Scale="1,1,1") {
    rotationMode=BMM_Additive
    rotationSpace=BCS_BoneSpace
    translationMode=BMM_Additive
    translationSpace=BCS_BoneSpace
    alphaBoolBlend {blendOption=Linear}
    alphaScaleBiasClamp {bInterpResult=true, interpSpeedIncreasing=5}
    boneToModify {boneName=dyn_tasset_l}
    translation {z=1}
   }
   n15 = "Transform (Modify) Bone - Bone: dyn_tasset_r"!(ComponentPose=@n14, bAlphaBoolEnabled:bool=Get bIsSkyDiving(), Translation="0,0.5,0.5", Rotation="0,5,0", Scale="1,1,1") -> (Pose) {
    alphaInputType=Bool
    rotationMode=BMM_Additive
    rotationSpace=BCS_BoneSpace
    translationMode=BMM_Additive
    translationSpace=BCS_BoneSpace
    alphaBoolBlend {blendInTime=0.1, blendOption=Linear, blendOutTime=0.1}
    alphaScaleBiasClamp {bInterpResult=true, interpSpeedIncreasing=5}
    boneToModify {boneName=dyn_tasset_r}
    rotation {yaw=5}
    translation {y=0.5, z=0.5}
   }
   n16 = "Transform (Modify) Bone - Bone: dyn_tasset_bk"!(ComponentPose=@n15, bAlphaBoolEnabled:bool=Get bIsSkydiveFloating(), Translation="0,0,0", Rotation="0,-10,0", Scale="1,1,1") -> (Pose) {
    alphaInputType=Bool
    rotationMode=BMM_Additive
    rotationSpace=BCS_BoneSpace
    translationMode=BMM_Additive
    translationSpace=BCS_BoneSpace
    alphaBoolBlend {blendInTime=0.1, blendOption=Linear, blendOutTime=0.1}
    alphaScaleBiasClamp {bInterpResult=true, interpSpeedIncreasing=5}
    boneToModify {boneName=dyn_tasset_bk}
    rotation {yaw=-10}
   }
   Output Pose(Result=Component To Local(ComponentPose=@n16)) {layerGroup=None}
   NOT Boolean(A=Get bIsCrouching())
  ]}
 EventGraph {nodes [
   Event BlueprintUpdateAnimation()
   > n1 = Sequence!(execute) -> (then_0, then_1, then_2:Exec)
   > Set Tasset fix(Tasset fix:double=1) -> (Output_Get:double)
   > n2 = Branch(Condition=OR Boolean(A=Get bIsCrouching(), B=Get bIsSkydiveFloating()))
   > Set Tasset BK(Tasset BK:bool=false) -> (Output_Get:bool)
   Set Tasset BK(execute=@n2.else, Tasset BK=true)
   n3 = PlayerMovement(execute=@n1.then_1, AnimInstance=Self-Reference(), XDotProductCoefficient=1, YDotProductCoefficient=1, ZDotProductCoefficient=1)
   > Set RightMovement(RightMovement:float=MapRangeClamped(Value=@n3, InRangeA=0, InRangeB=1)) -> (Output_Get:float)
   > Set LeftMovement(LeftMovement:float=MapRangeClamped(Value=@n3, InRangeA=0, InRangeB=-1)) -> (Output_Get:float)
   > Set BWDMovement(BWDMovement:float=MapRangeClamped(Value=@n3.YMovement, InRangeA=0, InRangeB=-1)) -> (Output_Get:float)
   n4 = Sequence!(execute=@n1.then_2) -> (then_0, then_1)
   n5 = Get BWDMovement() -> (BWDMovement:float)
   n6 = Get LeftMovement() -> (LeftMovement:float)
   > n7 = Branch(Condition=AND Boolean(A="float >= float"(A=@n5, B=0.5), B="float >= float"(A=@n6, B=0.5)))
   > Set BK_LeftMovement(BK_LeftMovement:float=1) -> (Output_Get:float)
   Set BK_LeftMovement(execute=@n7.else, BK_LeftMovement=0)
   n8 = Get RightMovement() -> (RightMovement:float)
   n9 = Get BWDMovement()
   n10 = Branch(execute=@n4.then_1, Condition=AND Boolean(A="float >= float"(A=@n9, B=0.5), B="float >= float"(A=@n8, B=0.5)))
   > Set BK_RightMovement(BK_RightMovement:float=1) -> (Output_Get:float)
   Set BK_RightMovement(execute=@n10.else, BK_RightMovement=0)
   n11 = Event OnBeginFalling()
   Set jumping(execute=[@n11.then, @n12.then], jumping:double=1) -> (Output_Get:double)
   n12 = Event OnBeginJumping()
   Event BlueprintInitializeAnimation()
   > Set Skel_Mesh(Skel_Mesh=GetOwnerSkeletalMeshComponent())
   n13 = Event OnEndFalling()
   Set jumping(execute=[@n13.then, @n14.then], jumping=0)
   n14 = Event OnEndJumping()
   Event OnEndCrouching()
   > Set crouching(crouching:double=0) -> (Output_Get:double)
   Event OnBeginCrouching()
   > Set crouching(crouching=1)
   GetAlphaFromReferenceTransformDelta(SkeletalMeshComponent=Get Skel_Mesh(), BoneName=spine_05, BaseName=pelvis, InputRange=(X=7,Y=20), TwistAxis="1,0,0")
  ]}
]
variables [
 BK_LeftMovement {type=number}
 BK_RightMovement {type=number}
 BWDMovement {type=number}
 crouching {type=number}
 jumping {type=number}
 LeftMovement {type=number}
 RightMovement {type=number}
 Skel_Mesh {type=SkeletalMeshComponent}
 Tasset BK {type=boolean}
 Tasset fix {type=number}
]
