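# Physics asset for the Possession medium male character's FaceAcc part, covering its dynamic hair bones with capsule bodies.
# Three back chains (left, center, right) and two front chains hang from the head, alongside capsules for pelvis, upper spine, head, clavicles and upper arms.
# Torso, arm and most chain-root joints swing and twist freely, while joints further along the hair have tighter angular limits, generally widening toward the back-hair tips.
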PhysicsAsset /BRCosmetics/Characters/Player/Male/Medium/Bodies/M_MED_Possession/Meshes/Parts/M_MED_Possession_FaceAcc_Physics
bodies {
 clavicle_l [{
   shapeType=Capsule
   length=8.6885
   radius=6.0166
   shapeName=clavicle_r_capsule
   center=<10.3187 -1.6222 2.637>
   rotation=<rot 39.9964 168.2484 -19.0754>
  }]
 clavicle_r [{
   shapeType=Capsule
   length=8.6885
   radius=6.0166
   shapeName=clavicle_r_capsule
   center=<-10.3146 1.3193 -2.5587>
   rotation=<rot 39.9964 168.2484 -19.0754>
  }]
 dyn_Hair_BackL_01 [{shapeType=Capsule, length=0.0017, radius=3.0581, shapeName=dyn_Hair_BackL_01_capsule}]
 dyn_Hair_BackL_02 [{
   shapeType=Capsule
   length=4.8415
   radius=6.1327
   shapeName=dyn_Hair_BackL_02_capsule
   center=<0 -2.4043 0.0611>
   rotation=<rot 0 -0.0001 91.4546>
  }]
 dyn_Hair_BackL_03 [{
   shapeType=Capsule
   length=12.6344
   radius=6.1612
   shapeName=dyn_Hair_BackR_04_capsule
   center=<6.8443 -5.3536 -0.8314>
   rotation=<rot 0.6137 -3.7205 -0.0514>
  }]
 dyn_Hair_BackL_04 [{
   shapeType=Capsule
   length=7.3225
   radius=5.6956
   shapeName=dyn_Hair_BackL_04_capsule
   center=<3.4189 -3.653 0.2019>
   rotation=<rot -1.4299 13.1845 89.9668>
  }]
 dyn_Hair_BackL_05 [{
   shapeType=Capsule
   length=5.5399
   radius=5.1272
   shapeName=dyn_Hair_BackL_05_capsule
   center=<0.8953 -2.4721 -0.0049>
   rotation=<rot -0.0001 0 91.3248>
  }]
 dyn_Hair_BackL_06 [{shapeType=Capsule, length=5.5399, radius=5.1272, shapeName=dyn_Hair_BackL_06_capsule, center=<1 -2.7602 0>, rotation=<rot -0.0001 0 91.3248>}]
 dyn_Hair_BackL_07 [{shapeType=Capsule, length=5.5399, radius=5.1272, shapeName=dyn_Hair_BackL_06_capsule, center=<1 -2.7602 0>, rotation=<rot -0.0001 0 91.3248>}]
 dyn_Hair_BackR_01 [{shapeType=Capsule, length=0.0017, radius=3.0581, shapeName=dyn_Hair_BackR_01_capsule}]
 dyn_Hair_BackR_02 [{
   shapeType=Capsule
   length=4.8415
   radius=6.1327
   shapeName=dyn_Hair_BackR_02_capsule
   center=<0 -2.4043 0.0611>
   rotation=<rot 0 -0.0001 91.4546>
  }]
 dyn_Hair_BackR_03 [{
   shapeType=Capsule
   length=12.6344
   radius=6.1612
   shapeName=dyn_Hair_BackR_04_capsule
   center=<6.3941 -3.4387 1.2938>
   rotation=<rot 2.9363 1.3506 -0.6399>
  }]
 dyn_Hair_BackR_04 [{
   shapeType=Capsule
   length=6.8637
   radius=6.1544
   shapeName=dyn_Hair_BackL_04_capsule
   center=<3.3755 -3.4699 0.1818>
   rotation=<rot -1.4299 13.1846 96.0799>
  }]
 dyn_Hair_BackR_05 [{
   shapeType=Capsule
   length=5.3991
   radius=5.268
   shapeName=dyn_Hair_BackR_05_capsule
   center=<0.4718 -0.7181 0.5633>
   rotation=<rot -0.0509 2.1966 91.3237>
  }]
 dyn_Hair_BackR_06 [{shapeType=Capsule, length=5.5399, radius=5.1272, shapeName=dyn_Hair_BackR_06_capsule, center=<1 -3.2918 0>, rotation=<rot -0.0001 0 91.3248>}]
 dyn_Hair_BackR_07 [{
   shapeType=Capsule
   length=6.0399
   radius=4.6272
   shapeName=dyn_Hair_Back_06_capsule
   center=<1 -2.5817 0>
   rotation=<rot 0.0989 -4.2827 91.321>
  }]
 dyn_Hair_BackR_08 [{
   shapeType=Capsule
   length=5.5399
   radius=5.1272
   shapeName=dyn_Hair_Back_06_capsule
   center=<1 -1.6073 0>
   rotation=<rot 0.0989 -4.2827 91.321>
  }]
 dyn_Hair_Back_01 [{shapeType=Capsule, length=0.0017, radius=3.0581, shapeName=dyn_Hair_Back_01_capsule}]
 dyn_Hair_Back_02 [{
   shapeType=Capsule
   length=4.8415
   radius=6.1327
   shapeName=dyn_Hair_Back_02_capsule
   center=<0 -2.4043 0.0611>
   rotation=<rot 0 -0.0001 91.4546>
  }]
 dyn_Hair_Back_03 [{shapeType=Capsule, length=12.6344, radius=6.1612, shapeName=dyn_Hair_Back_03_capsule, center=<5 -5 0>}]
 dyn_Hair_Back_04 [{
   shapeType=Capsule
   length=5.9422
   radius=5.3078
   shapeName=dyn_Hair_Back_04_capsule
   center=<1.7 -2.2639 0.5>
   rotation=<rot -0.1506 6.5248 91.3162>
  }]
 dyn_Hair_Back_05 [{shapeType=Capsule, length=5.5399, radius=5.1272, shapeName=dyn_Hair_Back_05_capsule, center=<1 -2.8087 0>, rotation=<rot -0.0001 0 91.3248>}]
 dyn_Hair_Back_06 [{
   shapeType=Capsule
   length=6.5399
   radius=4.1272
   shapeName=dyn_Hair_Back_06_capsule
   center=<1 -1.6073 0>
   rotation=<rot 0.0989 -4.2827 91.321>
  }]
 dyn_Hair_Back_07 [{
   shapeType=Capsule
   length=7.2899
   radius=3.3772
   shapeName=dyn_Hair_Back_06_capsule
   center=<1 -1.6073 0>
   rotation=<rot 0.0989 -4.2827 91.321>
  }]
 dyn_Hair_Back_08 [{
   shapeType=Capsule
   length=6.7899
   radius=3.8772
   shapeName=dyn_Hair_Back_06_capsule
   center=<1 -1.5 -2>
   rotation=<rot 0.0989 -4.2827 91.321>
  }]
 dyn_hair_l_Front_00 [{
   shapeType=Capsule
   length=0.505
   radius=0.505
   shapeName=dyn_hair_l_Front_00_capsule
   center=<3.1569 -0.1907 0.1744>
   rotation=<rot 0 0.057 -73.9322>
  }]
 dyn_hair_l_Front_01 [{shapeType=Capsule, length=2.75, radius=2.25, shapeName=dyn_hair_l_Front_01_capsule, center=<3.6882 0 0>, rotation=<rot 90 0 0>}]
 dyn_hair_l_Front_02 [{shapeType=Capsule, length=4.4563, radius=1.4436, shapeName=dyn_hair_l_Front_02_capsule, center=<3.5 0 0>, rotation=<rot 90 0 0>}]
 dyn_hair_l_Front_MSTR [{shapeType=Capsule, length=0.505, radius=0.505, shapeName=dyn_hair_l_Front_MSTR_capsule, rotation=<rot 0 -90 0>}]
 dyn_hair_r_F_02 [{shapeType=Capsule, length=2.375, radius=2.625, shapeName=dyn_hair_r_Front_02_capsule, center=<3.5 0 0>, rotation=<rot 90 0 0>}]
 dyn_hair_r_Fr_00 [{
   shapeType=Capsule
   length=3.9841
   radius=1.4813
   shapeName=dyn_hair_r_Fr_00_capsule
   center=<3.811 0.6276 0.0482>
   rotation=<rot 0 166.8948 -73.6535>
  }]
 dyn_hair_r_Fr_01 [{shapeType=Capsule, length=1.875, radius=3.125, shapeName=dyn_hair_r_Front_01_capsule, center=<3 0 0>, rotation=<rot 90 0 0>}]
 dyn_hair_r_Fr_03 [{shapeType=Capsule, length=2.375, radius=2.375, shapeName=dyn_hair_r_Front_03_capsule, center=<3.5 0 0>, rotation=<rot 90 0 0>}]
 dyn_hair_r_Fr_04 [{shapeType=Capsule, length=4.875, radius=1.875, shapeName=dyn_hair_r_Front_04_capsule, center=<3.5 0 0>, rotation=<rot 90 0 0>}]
 dyn_hair_r_Fr_MSTR [{shapeType=Capsule, length=0.505, radius=0.505, shapeName=dyn_hair_r_Fr_MSTR_capsule, rotation=<rot 0 -90 0>}]
 head [{shapeType=Capsule, length=4.29, radius=5.96, shapeName=head_capsule, center=<2.0202 0.1733 -0.2592>, rotation=<rot 90 -0.02 -0.0202>}]
 pelvis [{shapeType=Capsule, length=14.7443, radius=21.8397, shapeName=pelvis_capsule, center=<-9.1916 8.0491 0>}]
 spine_03 [{shapeType=Capsule, length=9.6221, radius=13.8579, shapeName=spine_03_capsule, center=<-8.2852 4.2209 0>}]
 spine_04 [{shapeType=Capsule, length=13.748, radius=10.9696, shapeName=spine_04_capsule, center=<1.2998 1.9478 0>}]
 spine_05 [{shapeType=Capsule, length=21.4339, radius=6.6011, shapeName=spine_05_capsule, center=<2.8604 0.2659 0.0001>}]
 upperarm_l [{shapeType=Capsule, length=42, radius=6.25, shapeName=upperarm_l_capsule, center=<25.2818 0 -2.8292>, rotation=<rot 83.6148 -0.0003 -0.0004>}]
 upperarm_r [{shapeType=Capsule, length=42, radius=6.25, shapeName=upperarm_l_capsule, center=<-24.8822 0 2.7845>, rotation=<rot 83.6148 -0.0003 -0.0004>}]
}
constraints [
 {
  bone1Name=head
  bone2Name=spine_05
  swing1LimitDegrees=45
  swing1Motion=Free
  swing2LimitDegrees=45
  swing2Motion=Free
  twistLimitDegrees=45
  twistMotion=Free
 }
 {
  bone1Name=spine_03
  bone2Name=pelvis
  swing1LimitDegrees=45
  swing1Motion=Free
  swing2LimitDegrees=45
  swing2Motion=Free
  twistLimitDegrees=45
  twistMotion=Free
 }
 {
  bone1Name=spine_04
  bone2Name=spine_03
  swing1LimitDegrees=45
  swing1Motion=Free
  swing2LimitDegrees=45
  swing2Motion=Free
  twistLimitDegrees=45
  twistMotion=Free
 }
 {
  bone1Name=spine_05
  bone2Name=spine_04
  swing1LimitDegrees=45
  swing1Motion=Free
  swing2LimitDegrees=45
  swing2Motion=Free
  twistLimitDegrees=45
  twistMotion=Free
 }
 {
  bone1Name=upperarm_l
  bone2Name=clavicle_l
  swing1LimitDegrees=45
  swing1Motion=Free
  swing2LimitDegrees=45
  swing2Motion=Free
  twistLimitDegrees=45
  twistMotion=Free
 }
 {
  bone1Name=upperarm_r
  bone2Name=clavicle_r
  swing1LimitDegrees=45
  swing1Motion=Free
  swing2LimitDegrees=45
  swing2Motion=Free
  twistLimitDegrees=45
  twistMotion=Free
 }
 {
  bone1Name=clavicle_r
  bone2Name=spine_05
  swing1LimitDegrees=45
  swing1Motion=Free
  swing2LimitDegrees=45
  swing2Motion=Free
  twistLimitDegrees=45
  twistMotion=Free
 }
 {
  bone1Name=clavicle_l
  bone2Name=spine_05
  swing1LimitDegrees=45
  swing1Motion=Free
  swing2LimitDegrees=45
  swing2Motion=Free
  twistLimitDegrees=45
  twistMotion=Free
 }
 {
  bone1Name=dyn_Hair_BackL_01
  bone2Name=head
  swing1LimitDegrees=45
  swing1Motion=Free
  swing2LimitDegrees=45
  swing2Motion=Free
  twistLimitDegrees=45
  twistMotion=Free
 }
 {
  bone1Name=dyn_Hair_BackL_02
  bone2Name=dyn_Hair_BackL_01
  swing1LimitDegrees=45
  swing1Motion=Free
  swing2LimitDegrees=45
  swing2Motion=Free
  twistLimitDegrees=45
  twistMotion=Free
 }
 {bone1Name=dyn_Hair_BackL_03, bone2Name=dyn_Hair_BackL_02, swing1LimitDegrees=5, swing2LimitDegrees=1, twistLimitDegrees=5}
 {bone1Name=dyn_Hair_BackL_04, bone2Name=dyn_Hair_BackL_03, swing1LimitDegrees=30, swing2LimitDegrees=1, twistLimitDegrees=10}
 {bone1Name=dyn_Hair_BackL_05, bone2Name=dyn_Hair_BackL_04, swing1LimitDegrees=30, swing2LimitDegrees=1, twistLimitDegrees=10}
 {bone1Name=dyn_Hair_BackL_06, bone2Name=dyn_Hair_BackL_05, swing1LimitDegrees=60, swing2LimitDegrees=10, twistLimitDegrees=10}
 {bone1Name=dyn_Hair_BackL_07, bone2Name=dyn_Hair_BackL_06, swing1LimitDegrees=60, swing2LimitDegrees=10, twistLimitDegrees=3}
 {
  bone1Name=dyn_Hair_BackR_01
  bone2Name=head
  swing1LimitDegrees=45
  swing1Motion=Free
  swing2LimitDegrees=45
  swing2Motion=Free
  twistLimitDegrees=45
  twistMotion=Free
 }
 {
  bone1Name=dyn_Hair_BackR_02
  bone2Name=dyn_Hair_BackR_01
  swing1LimitDegrees=45
  swing1Motion=Free
  swing2LimitDegrees=45
  swing2Motion=Free
  twistLimitDegrees=45
  twistMotion=Free
 }
 {bone1Name=dyn_Hair_BackR_03, bone2Name=dyn_Hair_BackR_02, swing1LimitDegrees=10, swing2LimitDegrees=1, twistLimitDegrees=15}
 {bone1Name=dyn_Hair_BackR_04, bone2Name=dyn_Hair_BackR_03, swing1LimitDegrees=30, swing2LimitDegrees=10, twistLimitDegrees=10}
 {bone1Name=dyn_Hair_BackR_05, bone2Name=dyn_Hair_BackR_04, swing1LimitDegrees=30, swing2LimitDegrees=10, twistLimitDegrees=10}
 {bone1Name=dyn_Hair_BackR_06, bone2Name=dyn_Hair_BackR_05, swing1LimitDegrees=30, swing2LimitDegrees=10, twistLimitDegrees=10}
 {bone1Name=dyn_Hair_BackR_07, bone2Name=dyn_Hair_BackR_06, swing1LimitDegrees=60, swing2LimitDegrees=10, twistLimitDegrees=3}
 {bone1Name=dyn_Hair_BackR_08, bone2Name=dyn_Hair_BackR_07, swing1LimitDegrees=60, swing2LimitDegrees=10, twistLimitDegrees=3}
 {
  bone1Name=dyn_Hair_Back_01
  bone2Name=head
  swing1LimitDegrees=45
  swing1Motion=Free
  swing2LimitDegrees=45
  swing2Motion=Free
  twistLimitDegrees=45
  twistMotion=Free
 }
 {
  bone1Name=dyn_Hair_Back_02
  bone2Name=dyn_Hair_Back_01
  swing1LimitDegrees=45
  swing1Motion=Free
  swing2LimitDegrees=45
  swing2Motion=Free
  twistLimitDegrees=45
  twistMotion=Free
 }
 {bone1Name=dyn_Hair_Back_03, bone2Name=dyn_Hair_Back_02, swing1LimitDegrees=5, swing2LimitDegrees=10, twistLimitDegrees=5}
 {bone1Name=dyn_Hair_Back_04, bone2Name=dyn_Hair_Back_03, swing1LimitDegrees=30, swing2LimitDegrees=10, twistLimitDegrees=10}
 {bone1Name=dyn_Hair_Back_05, bone2Name=dyn_Hair_Back_04, swing1LimitDegrees=30, swing2LimitDegrees=10, twistLimitDegrees=10}
 {bone1Name=dyn_Hair_Back_06, bone2Name=dyn_Hair_Back_05, swing1LimitDegrees=30, swing2LimitDegrees=10, twistLimitDegrees=10}
 {bone1Name=dyn_Hair_Back_07, bone2Name=dyn_Hair_Back_06, swing1LimitDegrees=60, swing2LimitDegrees=10, twistLimitDegrees=3}
 {bone1Name=dyn_Hair_Back_08, bone2Name=dyn_Hair_Back_07, swing1LimitDegrees=60, swing2LimitDegrees=10, twistLimitDegrees=3}
 {
  bone1Name=dyn_hair_r_Fr_00
  bone2Name=dyn_hair_r_Fr_MSTR
  swing1LimitDegrees=45
  swing1Motion=Free
  swing2LimitDegrees=45
  swing2Motion=Free
  twistLimitDegrees=45
  twistMotion=Free
 }
 {bone1Name=dyn_hair_r_Fr_01, bone2Name=dyn_hair_r_Fr_00, swing1LimitDegrees=35, swing2LimitDegrees=15, twistLimitDegrees=5}
 {bone1Name=dyn_hair_r_F_02, bone2Name=dyn_hair_r_Fr_01, swing1LimitDegrees=35, swing2LimitDegrees=15, twistLimitDegrees=5}
 {bone1Name=dyn_hair_r_Fr_03, bone2Name=dyn_hair_r_F_02, swing1LimitDegrees=35, swing2LimitDegrees=15, twistLimitDegrees=5}
 {bone1Name=dyn_hair_r_Fr_04, bone2Name=dyn_hair_r_Fr_03, swing1LimitDegrees=35, swing2LimitDegrees=15, twistLimitDegrees=5}
 {bone1Name=dyn_hair_l_Front_00, bone2Name=dyn_hair_l_Front_MSTR, swing1LimitDegrees=35, swing2LimitDegrees=15, twistLimitDegrees=5}
 {
  bone1Name=dyn_hair_l_Front_MSTR
  bone2Name=head
  swing1LimitDegrees=45
  swing1Motion=Free
  swing2LimitDegrees=45
  swing2Motion=Free
  twistLimitDegrees=45
  twistMotion=Free
 }
 {bone1Name=dyn_hair_l_Front_01, bone2Name=dyn_hair_l_Front_00, swing1LimitDegrees=35, swing2LimitDegrees=15, twistLimitDegrees=5}
 {bone1Name=dyn_hair_l_Front_02, bone2Name=dyn_hair_l_Front_01, swing1LimitDegrees=35, swing2LimitDegrees=15, twistLimitDegrees=5}
 {
  bone1Name=dyn_hair_r_Fr_MSTR
  bone2Name=head
  swing1LimitDegrees=45
  swing1Motion=Free
  swing2LimitDegrees=45
  swing2Motion=Free
  twistLimitDegrees=45
  twistMotion=Free
 }
]
